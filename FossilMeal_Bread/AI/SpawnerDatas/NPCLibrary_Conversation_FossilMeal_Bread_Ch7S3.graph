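# Conversation database for the FossilMeal Bread NPC (Season 41 entry tag), offering greeting dialogue with hire, duel and quest options.
# An opening Speech node branches into a link to the shared NPCLibrary quests conversation plus DUEL and HIRE preview choices.
# Duel and hire choices are gated by the NPC lacking NoDuel or NoHire tags, then lead to confirm (ChallengeNPC, HireNPC) or back.
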ConversationDatabase /FossilMeal_Bread/AI/SpawnerDatas/NPCLibrary_Conversation_FossilMeal_Bread_Ch7S3
entry_points [{tag=Conversation.Entry.Season41.FossilMeal.Bread, destinations [#1]}]
nodes [
 ConversationEntryPointNode_7 {
  id=#1
  class=ConversationEntryPointNode
  outputs [#2]
  properties {entryTag {tagName=Conversation.Entry.Season41.FossilMeal.Bread}}
 }
 Speech_C_4 {
  id=#2
  class=Speech_C
  outputs [#3, #4, #5]
  properties {
   nodeName=""
   generalConfig {
    defaultMessage=Hmm?
    contextualMessages [
     {message=Hmm?, contextRequirements [], priority {curve {curveTable=null, rowName=None}}}
     {message=Do we have mutual business?, contextRequirements [], priority {curve {curveTable=null, rowName=None}}}
    ]
   }
  }
 }
 Speech_C_0 {
  id=#5
  class=Speech_C
  outputs [#6, #7]
  properties {
   bUsePerNPC_ContextualMessageLookup=true
   softTaskIcon=/Game/Athena/HUD/Interaction/T_UI_NPC_Hire
   generalConfig {
    defaultMessage=Make this alliance worth my while.
    contextualMessages [
     {message=Make this alliance worth my while., contextRequirements [], priority {curve {curveTable=null, rowName=None}}}
     {message=I'll work with you... for the right price., contextRequirements [], priority {curve {curveTable=null, rowName=None}}}
    ]
   }
  }
  sub_nodes [
   FortConversationChoiceWithPreviewNode_0 {
    class=FortConversationChoiceWithPreviewNode
    properties {defaultChoiceDisplayText=HIRE, choiceTags {gameplayTags [{tagName=Conversation.Choice.CustomSortPriority.Lowest}]}}
   }
   FortConversationRequirement_ControllerRequirement_1 {
    class=FortConversationRequirement_ControllerRequirement
    properties {
     nodeName=Require: Has GameplayTag
     participantId {tagName=Conversation.Participant.NPC}
     requirement {class=FortControllerRequirement_HasOwnedTag, properties {bInverseCondition=true, ownedTag {tagName=NPC.Status.NoHire}}}
    }
   }
  ]
 }
 HireNPC_C_0 {
  id=#6
  class=HireNPC_C
  outputs []
  sub_nodes [ConversationChoiceNode_4 {class=ConversationChoiceNode, properties {defaultChoiceDisplayText=CONFIRM}}]
 }
 Back_C_3 {
  id=#7
  class=Back_C
  sub_nodes [ConversationChoiceNode_1 {class=ConversationChoiceNode, properties {defaultChoiceDisplayText=BACK}}]
 }
 Speech_C_5 {
  id=#4
  class=Speech_C
  outputs [#8, #9]
  properties {
   bUsePerNPC_ContextualMessageLookup=true
   softTaskIcon=/Game/UI/Icons/Icon-Stat/Icon-WeaponSkill-XL
   generalConfig {
    defaultMessage="Finally, some action."
    contextualMessages [
     {message="Finally, some action.", contextRequirements [], priority {curve {curveTable=null, rowName=None}}}
     {message=My claws need sharpening., contextRequirements [], priority {curve {curveTable=null, rowName=None}}}
    ]
   }
  }
  sub_nodes [
   FortConversationChoiceWithPreviewNode_4 {
    class=FortConversationChoiceWithPreviewNode
    properties {defaultChoiceDisplayText=DUEL, choiceTags {gameplayTags [{tagName=Conversation.Choice.CustomSortPriority.Low}]}}
   }
   FortConversationRequirement_ControllerRequirement_0 {
    class=FortConversationRequirement_ControllerRequirement
    properties {
     nodeName=Require: Has GameplayTag
     participantId {tagName=Conversation.Participant.NPC}
     requirement {class=FortControllerRequirement_HasOwnedTag, properties {bInverseCondition=true, ownedTag {tagName=npc.status.noduel}}}
    }
   }
  ]
 }
 ChallengeNPC_C_1 {
  id=#8
  class=ChallengeNPC_C
  outputs []
  sub_nodes [ConversationChoiceNode_7 {class=ConversationChoiceNode, properties {defaultChoiceDisplayText=CONFIRM}}]
 }
 Back_C_5 {
  id=#9
  class=Back_C
  sub_nodes [ConversationChoiceNode_5 {class=ConversationChoiceNode, properties {defaultChoiceDisplayText=BACK}}]
 }
 ConversationLinkNode_1 {id=#3, class=ConversationLinkNode, properties {remoteEntryTag {tagName=Conversation.Entry.NPCLibrary.Quests}}}
]
